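# Physics asset for the face accessory mesh of the male SpeedDialBattle WhiteCap character, setting up physical motion for its bandana.
# Left and right two-bone bandana chains hang off the head through constraints with limited swing and very tight twist.
# Capsules cover the bandana bones, plus the head, neck and upper spine, which have no constraints among themselves.
# Left and right tips differ in capsule size and swing limits.
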PhysicsAsset /BRCosmetics/Characters/Player/Male/Medium/Bodies/M_MED_SpeedDialBattle_WhiteCap/Meshes/Parts/M_MED_SpeedDialBattle_WhiteCap_FaceAcc_Physics
bodies {
 dyn_Bandana_l_02_jnt [{
   shapeType=Capsule
   length=2.01
   radius=1.3409
   shapeName=dyn_Bandana_r_02_jnt_capsule
   center=<1.6425 -0.0182 0.2042>
   rotation=<rot 0 87.357 91.9922>
  }]
 dyn_Bandana_l_03_jnt [{
   shapeType=Capsule
   length=2.9662
   radius=1.763
   shapeName=dyn_Bandana_r_03_jnt_capsule
   center=<2.605 -0.9129 0>
   rotation=<rot 90 14.0362 -149.3452>
  }]
 dyn_Bandana_r_02_jnt [{
   shapeType=Capsule
   length=2.01
   radius=1.3409
   shapeName=dyn_Bandana_r_02_jnt_capsule
   center=<1.6425 -0.0182 0.2042>
   rotation=<rot 0 87.357 -88.0078>
  }]
 dyn_Bandana_r_03_jnt [{
   shapeType=Capsule
   length=1.7356
   radius=1.5178
   shapeName=dyn_Bandana_r_03_jnt_capsule
   center=<2.1268 -0.7701 0>
   rotation=<rot 90 0 16.6185>
  }]
 head [{shapeType=Capsule, length=5.7069, radius=7.641, center=<5.6117 0.5265 0.14>, rotation=<rot 0.2845 -86.8219 -89.6285>}]
 neck_02 [{shapeType=Capsule, length=5.3921, radius=4.5449, shapeName=neck_02_capsule, center=<0.3227 -1.4811 0>, rotation=<rot 90 0 0>}]
 spine_05 [
  {shapeType=Capsule, length=4.3272, radius=8.6584, shapeName=spine_05_capsule, center=<8.9488 0.4868 0.0148>, rotation=<rot 0 -90 0>}
  {shapeType=Capsule, length=7.6188, radius=5.8147, shapeName=spine_05_capsule, center=<20.6657 2.4856 0>}
 ]
}
constraints [
 {bone1Name=dyn_Bandana_r_03_jnt, bone2Name=dyn_Bandana_r_02_jnt, swing1LimitDegrees=22, swing2LimitDegrees=40, twistLimitDegrees=1}
 {bone1Name=dyn_Bandana_r_02_jnt, bone2Name=head, swing1LimitDegrees=10, swing2LimitDegrees=40, twistLimitDegrees=3}
 {bone1Name=dyn_Bandana_l_03_jnt, bone2Name=dyn_Bandana_l_02_jnt, swing1LimitDegrees=10, swing2LimitDegrees=40, twistLimitDegrees=3}
 {bone1Name=dyn_Bandana_l_02_jnt, bone2Name=head, swing1LimitDegrees=10, swing2LimitDegrees=40, twistLimitDegrees=3}
]
