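# Post-process style face accessory AnimBP for the SpeedDialBattle White character, animating the left bandana offset bone with secondary motion.
# Two chains copy the parent pose through head-limited AnimDynamics, RigidBody and additive corrective transforms, selected by whether a custom part mesh name ends in "jacket".
# Head swing angle drives corrective alphas, with ranges varying by crouch, movement and targeting state.
# A final additive rotation applies while skydiving.
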AnimBlueprint /BRCosmetics/Characters/Player/Male/Medium/Bodies/M_MED_SpeedDialBattle_White/Meshes/Parts/M_MED_SpeedDialBattle_White_FaceAcc_AnimBP
target_skeleton=&/M_MED_SpeedDialBattle_White_FaceAcc_Skeleton
graphs [
 AnimGraph {nodes [
   n1 = "Transform (Modify) Bone - Bone: Bandana_l_offset_jnt"!(
    ComponentPose=Local To Component(LocalPose=Use cached pose 'Dyn_cache_pose'())
    bAlphaBoolEnabled:bool=Get bIsSkyDiving()
    Translation="0,0,0"
    Rotation="-0.019,-49.128,0.088"
    Scale="1,1,1"
   ) -> (Pose) {
    alphaInputType=Bool
    rotationMode=BMM_Additive
    rotationSpace=BCS_BoneSpace
    translationMode=BMM_Additive
    translationSpace=BCS_BoneSpace
    alphaBoolBlend {blendOption=Linear}
    boneToModify {boneName=Bandana_l_offset_jnt}
    rotation=<rot -0.019 -49.128 0.088>
   }
   Output Pose(Result=Component To Local(ComponentPose=@n1)) {layerGroup=None}
   n2 = Anim Dynamics - Bone: Bandana_l_offset_jnt(ComponentPose=Local To Component(LocalPose=Copy Pose From Mesh() {bUseAttachedParent=true}), Alpha=0.5, ExternalForce="0,0,0") {
    angularDampingOverride=0.999
    angularSpringConstant=1
    bAngularSpring=true
    bGravityOverrideInSimSpace=true
    bOverrideAngularDamping=true
    bUseGravityOverride=true
    bUseSphericalLimits=true
    alphaBoolBlend {blendOption=Linear}
    alphaScaleBiasClamp {bClampResult=true}
    boundBone {boneName=Bandana_l_offset_jnt}
    componentLinearAccScale=<0.2 0.2 0.2>
    componentLinearVelScale=<1 1 1>
    gravityOverride {z=-300}
    physicsBodyDefinitions [{
      boundBone {boneName=Bandana_l_offset_jnt}
      boxExtents=<20 20 20>
      constraintSetup {angularLimitsMax=<15 20 5>, angularLimitsMin=<-15 -20 -5>, angularTarget=<27 0 0>}
      localJointOffset=<4 0 0>
     }]
    planarLimits [
     {drivingBone {boneName=head}, planeTransform {location=<0 -9 0>, rotation=<rot 0 5.0001 90>}}
     {drivingBone {boneName=head}, planeTransform {location=<0 -9 0>, rotation=<rot 0 5 45>}}
     {drivingBone {boneName=head}, planeTransform {location=<0 -9 0>, rotation=<rot 0 5 -45>}}
    ]
    relativeSpaceBone {boneName=spine_05}
    retargetingSettings {rotationComponent=TwistAngle, twistAxis=<0 -1>}
   }
   n3 = RigidBody!(ComponentPose=@n2, bAlphaBoolEnabled:bool, ExternalForce) -> (Pose) {
    alphaInputType=Bool
    bOverrideWorldGravity=true
    overridePhysicsAsset=&/M_MED_SpeedDialBattle_WhiteCap_FaceAcc_Physics
    simulationSpace=BaseBoneSpace
    alphaBoolBlend {blendInTime=0.1, blendOption=Linear, blendOutTime=0.1}
    baseBoneRef {boneName=Bandana_l_offset_jnt}
    componentAppliedLinearAccClamp=<1000 1000 1000>
    componentLinearAccScale=<0.1 0.5 0.2>
    componentLinearVelScale=<0.5 0.5 0.5>
    overrideWorldGravity {z=-650}
    simSpaceSettings {maxAngularAcceleration=1000, maxAngularVelocity=3, maxLinearAcceleration=1000, maxLinearVelocity=10}
   }
   n4 = "Transform (Modify) Bone - Bone: Bandana_l_offset_jnt"!(
    ComponentPose=@n3
    Alpha:float=Get rot_head_value() -> (rot_head_value:double)
    Translation="-0.242,-0.5,-0.752"
    Rotation="-3.457,-12,5.096"
    Scale="0.953,1.000,1.000"
   ) -> (Pose) {
    rotationMode=BMM_Additive
    rotationSpace=BCS_BoneSpace
    scaleMode=BMM_Additive
    scaleSpace=BCS_BoneSpace
    translationMode=BMM_Additive
    translationSpace=BCS_BoneSpace
    alphaBoolBlend {blendOption=Linear}
    boneToModify {boneName=Bandana_l_offset_jnt}
    rotation=<rot -3.457 -12 5.096>
    scale {x=0.9531}
    translation=<-0.242 -0.5 -0.752>
   }
   n5 = "Transform (Modify) Bone - Bone: Bandana_l_offset_jnt"(
    ComponentPose=@n4
    Alpha=Get rot_head_value_max() -> (rot_head_value_max:double)
    Translation="-1.5,-1,0"
    Rotation="-3.457,-45,5.096"
    Scale="0.953,1.000,1.000"
   ) {
    rotationMode=BMM_Additive
    rotationSpace=BCS_BoneSpace
    scaleMode=BMM_Additive
    scaleSpace=BCS_BoneSpace
    translationMode=BMM_Additive
    translationSpace=BCS_BoneSpace
    alphaBoolBlend {blendOption=Linear}
    boneToModify {boneName=Bandana_l_offset_jnt}
    rotation=<rot -3.457 -45 5.096>
    scale {x=0.9531}
    translation=<-1.5 -1>
   }
   n6 = Anim Dynamics - Bone: Bandana_l_offset_jnt(ComponentPose=Local To Component(LocalPose=Copy Pose From Mesh() {bUseAttachedParent=true}), Alpha=0.5, ExternalForce="0,0,0") {
    angularDampingOverride=0.999
    angularSpringConstant=1
    bAngularSpring=true
    bGravityOverrideInSimSpace=true
    bOverrideAngularDamping=true
    bUseGravityOverride=true
    bUseSphericalLimits=true
    alphaBoolBlend {blendOption=Linear}
    alphaScaleBiasClamp {bClampResult=true}
    boundBone {boneName=Bandana_l_offset_jnt}
    componentLinearAccScale=<0.2 0.2 0.2>
    componentLinearVelScale=<1 1 1>
    gravityOverride {z=-300}
    physicsBodyDefinitions [{
      boundBone {boneName=Bandana_l_offset_jnt}
      boxExtents=<20 20 20>
      constraintSetup {angularLimitsMax=<15 20 5>, angularLimitsMin=<-15 -20 -5>, angularTarget=<27 0 0>}
      localJointOffset=<4 0 0>
     }]
    planarLimits [
     {drivingBone {boneName=head}, planeTransform {location=<0 -9 0>, rotation=<rot 0 5.0001 90>}}
     {drivingBone {boneName=head}, planeTransform {location=<0 -9 0>, rotation=<rot 0 5 45>}}
     {drivingBone {boneName=head}, planeTransform {location=<0 -9 0>, rotation=<rot 0 5 -45>}}
    ]
    relativeSpaceBone {boneName=spine_05}
    retargetingSettings {rotationComponent=TwistAngle, twistAxis=<0 -1>}
   }
   n7 = RigidBody(ComponentPose=@n6) {
    alphaInputType=Bool
    bOverrideWorldGravity=true
    overridePhysicsAsset=/BRCosmetics/Characters/Player/Male/Medium/Bodies/M_MED_SpeedDialBattle_WhiteCap/Meshes/Parts/M_MED_SpeedDialBattle_WhiteCap_FaceAcc_Physics
    simulationSpace=BaseBoneSpace
    alphaBoolBlend {blendInTime=0.1, blendOption=Linear, blendOutTime=0.1}
    baseBoneRef {boneName=Bandana_l_offset_jnt}
    componentAppliedLinearAccClamp=<1000 1000 1000>
    componentLinearAccScale=<0.1 0.5 0.2>
    componentLinearVelScale=<0.5 0.5 0.5>
    overrideWorldGravity {z=-650}
    simSpaceSettings {maxAngularAcceleration=1000, maxAngularVelocity=3, maxLinearAcceleration=1000, maxLinearVelocity=10}
   }
   n8 = "Transform (Modify) Bone - Bone: Bandana_l_offset_jnt"(ComponentPose=@n7, Alpha=Get rot_head_value(), Translation="0,0,0", Rotation="9.168,-6.816,18.395", Scale="1,1,1") {
    rotationMode=BMM_Additive
    rotationSpace=BCS_BoneSpace
    scaleMode=BMM_Additive
    scaleSpace=BCS_BoneSpace
    translationMode=BMM_Additive
    translationSpace=BCS_BoneSpace
    alphaBoolBlend {blendOption=Linear}
    boneToModify {boneName=Bandana_l_offset_jnt}
    rotation=<rot 9.168 -6.816 18.395>
   }
   Save cached pose 'Dyn_cache_pose'(Pose=Blend Poses by bool(
     BlendPose_0=Component To Local(ComponentPose=@n5)
     BlendPose_1=Component To Local(ComponentPose=@n8)
     bActiveValue=Get switch_part() -> (switch_part:bool)
    ) {blendType=Linear}) {cacheName=Dyn_cache_pose}
  ]}
 EventGraph {nodes [
   n1 = Event OnCharacterCustomizationLoaded()
   n2 = Get SkeletalMesh!(self:Skinned Mesh Component Object Reference=Get Array Item with EFort Custom Part Type(Array=@n1.CustomMeshParts)) -> (SkeletalMesh:Skeletal Mesh Object Reference)
   > n3 = Set switch_part(switch_part:bool=EndsWith(SourceString=GetObjectName(Object=@n2), InSuffix=jacket)) -> (Output_Get:bool)
   > Set switch_part_not(switch_part_not:bool="Not Equal (Boolean)"(A=@n3, B=true) -> (ReturnValue=~)) -> (Output_Get:bool)
   Event BlueprintUpdateAnimation()
   > n4 = Sequence!(execute) -> (then_0, then_1, then_2:Exec, then_3:Exec)
   n5 = GetOwnerSkeletalMeshComponent()
   > Set if_Skydiving(if_Skydiving:double=GetSocketTransform!(self=@n5, InSocketName=root, TransformSpace=RTS_Component) -> (
     ReturnValue_Location_X:double
     ReturnValue_Location_Y:double
     ReturnValue_Location_Z:double
     ReturnValue_Rotation:Rotator
     ReturnValue_Scale:Vector
    )) -> (Output_Get:double)
   Set Return Value Location X(execute=@n4.then_2, Return Value Location X:double=GetSocketTransform(self=@n5, InSocketName=head, TransformSpace=RTS_Component)) -> (Output_Get:double)
   n6 = Get bIsMoving2D()
   n7 = Get bIsCrouching()
   n8 = Branch(execute=@n4.then_3, Condition=AND Boolean(A=AND Boolean(A=@n7, B=@n6), B=Get bIsTargeting()))
   n9 = GetTwistAndSwingAngleOfDeltaRotationFromRefPose(self=@n5, BoneName=head)
   > Set rot_head_value_max(rot_head_value_max:double=MapRangeClamped(Value=@n9.OutSwingAngle, InRangeA=52, InRangeB=60)) -> (Output_Get:double)
   n10 = Branch(execute=@n8.else, Condition="Not Equal (Boolean)"(A=@n7, B=true) -> (ReturnValue=~))
   > Set rot_head_value(rot_head_value:double=MapRangeClamped(Value=@n9.OutSwingAngle, InRangeA=35, InRangeB=44)) -> (Output_Get:double)
   Set rot_head_value(execute=@n10.else, rot_head_value=MapRangeClamped(Value=@n9.OutSwingAngle, InRangeA=39, InRangeB=41))
  ]}
]
variables [
 if_Skydiving {type=number}
 Return Value Location X {type=number}
 rot_head_value {type=number}
 rot_head_value_max {type=number}
 switch_part {type=boolean}
 switch_part_not {type=boolean}
]
